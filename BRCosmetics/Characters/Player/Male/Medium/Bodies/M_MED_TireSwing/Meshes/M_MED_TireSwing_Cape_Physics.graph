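# Physics asset for the cape on the TireSwing medium male player body cosmetic, pairing cape capsules with capsules on body and root bones.
# Four six-bone cape chains hang from master bones on the clavicles and upper spine, with per-joint swing limits and zero twist.
# Angularly free cross-constraints tie some neighboring chains together, mostly lower down, so they move as a sheet.
# Body joints and master-bone attachments are also angularly free.
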PhysicsAsset /BRCosmetics/Characters/Player/Male/Medium/Bodies/M_MED_TireSwing/Meshes/M_MED_TireSwing_Cape_Physics
bodies {
 calf_l [{shapeType=Capsule, length=36.3683, radius=13.9668, shapeName=calf_l_capsule, center=<-18.4993 0 0>, rotation=<rot 90 0 0>}]
 calf_r [{shapeType=Capsule, length=36.3683, radius=13.9668, shapeName=calf_r_capsule, center=<18.4993 0 0>, rotation=<rot 90 0 0>}]
 cape_01_l [{shapeType=Capsule, length=14.5431, radius=5.4275, shapeName=cape_01_l_capsule, center=<11.9983 0 0>, rotation=<rot 90 0 0>}]
 cape_01_r [{shapeType=Capsule, length=9.6914, radius=5.0845, shapeName=cape_01_r_capsule, center=<8.2303 0 0>, rotation=<rot 90 0 0>}]
 cape_02_l [
  {shapeType=Capsule, length=9.564, radius=7.6988, shapeName=cape_02_l_capsule, center=<21.5996 0 0>, rotation=<rot 0 0 90>}
  {shapeType=Capsule, length=14.1165, radius=6.0281, shapeName=cape_02_l_capsule, center=<6.364 0 0>, rotation=<rot 0 0 90>}
 ]
 cape_02_r [
  {shapeType=Capsule, length=18, radius=6, shapeName=cape_02_r_capsule, center=<5 0 0>, rotation=<rot 90 0 90>}
  {shapeType=Capsule, length=16.75, radius=7.25, shapeName=cape_02_r_capsule, center=<20 0 0>, rotation=<rot 90 0 90>}
 ]
 cape_03_l [{shapeType=Capsule, length=8.9831, radius=8.3946, shapeName=cape_03_l_capsule, center=<11.0576 0 0>, rotation=<rot 90 0 90>}]
 cape_03_r [{shapeType=Capsule, length=12.0209, radius=8.7929, shapeName=cape_03_r_capsule, center=<8 0 0>, rotation=<rot 0 0 90>}]
 cape_04_l [{shapeType=Capsule, length=5.3928, radius=9.2416, shapeName=cape_04_l_capsule, center=<9.4987 0 0>, rotation=<rot 0 0 90>}]
 cape_04_r [{shapeType=Capsule, length=8.6944, radius=13.1524, shapeName=cape_04_r_capsule, center=<8 -3.6409 8.0053>, rotation=<rot 90 0 0>}]
 cape_05_l [{shapeType=Capsule, length=4.2161, radius=8.8219, shapeName=cape_05_l_capsule, center=<7.655 0 0>, rotation=<rot 90 0 90>}]
 cape_05_r [{shapeType=Capsule, length=3.5813, radius=10.9912, shapeName=cape_05_r_capsule, center=<8 -0.0676 6.4656>, rotation=<rot 0 0 90>}]
 cape_06_l [{shapeType=Capsule, length=4.2637, radius=10.3256, shapeName=cape_06_l_capsule, center=<7.655 0 0>, rotation=<rot 0 0 90>}]
 cape_06_r [{shapeType=Capsule, length=3.4943, radius=11.0421, shapeName=cape_06_r_capsule, center=<8 -0.0676 6.4656>, rotation=<rot 0 0 90>}]
 cape_l_MSTR [{shapeType=Capsule, length=1, radius=1, shapeName=cape_l_MSTR_capsule, rotation=<rot 0 -90 0>}]
 cape_mid_01_l [{
   shapeType=Capsule
   length=13.2133
   radius=5.885
   shapeName=cape_mid_01_l_capsule
   center=<10.1787 0 4.251>
   rotation=<rot 87.5681 -0.0001 179.9999>
  }]
 cape_mid_01_r [{
   shapeType=Capsule
   length=13.2133
   radius=5.885
   shapeName=cape_mid_01_r_capsule
   center=<10.1787 0 4.251>
   rotation=<rot 87.5681 0.0002 -179.9998>
  }]
 cape_mid_02_l [{shapeType=Capsule, length=15.2976, radius=5.525, shapeName=cape_mid_02_l_capsule, center=<15.4392 0 0>, rotation=<rot 90 0 0>}]
 cape_mid_02_r [{shapeType=Capsule, length=15.2976, radius=5.525, shapeName=cape_mid_02_r_capsule, center=<15.4392 0 0>, rotation=<rot 90 0 0>}]
 cape_mid_03_l [{shapeType=Capsule, length=12.6579, radius=9.025, shapeName=cape_mid_03_l_capsule, center=<11.0576 0 0>, rotation=<rot 0 0 90>}]
 cape_mid_03_r [{shapeType=Capsule, length=12.6579, radius=9.025, shapeName=cape_mid_03_r_capsule, center=<11.0576 0 0>, rotation=<rot 0 0 90>}]
 cape_mid_04_l [{shapeType=Capsule, length=13.7686, radius=10.107, shapeName=cape_mid_04_l_capsule, center=<9.4987 0 4.2>, rotation=<rot 90 180 -90>}]
 cape_mid_04_r [{shapeType=Capsule, length=12.9271, radius=10.107, shapeName=cape_mid_04_r_capsule, center=<9.4987 0 4.2>, rotation=<rot 90 180 -90>}]
 cape_mid_05_l [{shapeType=Capsule, length=11.4499, radius=12.015, shapeName=cape_mid_05_l_capsule, center=<7.655 0 6.8334>, rotation=<rot 90 180 -90>}]
 cape_mid_05_r [{shapeType=Capsule, length=10.9573, radius=12.5075, shapeName=cape_mid_05_r_capsule, center=<7.655 0 6.8334>, rotation=<rot 90 180 -90>}]
 cape_mid_06_l [{shapeType=Capsule, length=15.4991, radius=13.6418, shapeName=cape_mid_06_l_capsule, center=<9.6942 0 5.3729>, rotation=<rot 0 0 90>}]
 cape_mid_06_r [{shapeType=Capsule, length=15.5122, radius=13.6286, shapeName=cape_mid_06_r_capsule, center=<9.6942 0 5.3729>, rotation=<rot 0 0 90>}]
 cape_mid_l_MSTR [{shapeType=Capsule, length=1, radius=1, shapeName=cape_mid_l_MSTR_capsule, rotation=<rot 0 -90 0>}]
 cape_mid_r_MSTR [{shapeType=Capsule, length=1, radius=1, shapeName=cape_mid_r_MSTR_capsule, rotation=<rot 0 -90 0>}]
 cape_r_MSTR [{shapeType=Capsule, length=1, radius=1, shapeName=cape_r_MSTR_capsule, rotation=<rot 0 -90 0>}]
 clavicle_l [{shapeType=Capsule, length=12.0737, radius=7.5263, shapeName=clavicle_l_capsule, center=<9.3 0 0>, rotation=<rot 81 0 0>}]
 clavicle_r [{shapeType=Capsule, length=12.0737, radius=7.5263, shapeName=clavicle_r_capsule, center=<-9.3 0 0>, rotation=<rot 81 0 0>}]
 foot_l [{shapeType=Capsule, length=16.1864, radius=13.9127, shapeName=foot_l_capsule, center=<-4.0851 7.4016 0>, rotation=<rot 90 180 -90>}]
 foot_r [{shapeType=Capsule, length=14.746, radius=13.9127, shapeName=foot_r_capsule, center=<5.1029 -6.88 0>, rotation=<rot 90 180 -90>}]
 lowerarm_l [{shapeType=Capsule, length=23.4969, radius=9.0028, shapeName=lowerarm_l_capsule, center=<10.6887 0 5.4647>, rotation=<rot 90 0 0>}]
 lowerarm_r [{shapeType=Capsule, length=14.7082, radius=9.6516, shapeName=lowerarm_r_capsule, center=<-18.4993 -3.3146 0>, rotation=<rot 90 0 0>}]
 pelvis [{shapeType=Capsule, length=11.546, radius=12.8138, shapeName=pelvis_capsule}]
 root [
  {
   shapeType=Capsule
   length=89.2237
   radius=17.4796
   shapeName=root_capsule
   center=<-36.9579 1.2052 66.2044>
   rotation=<rot 90 -178.2101 171.1006>
  }
  {shapeType=Capsule, length=119.6101, radius=15, shapeName=root_capsule, center=<-37.9537 -1.8346 29.3211>, rotation=<rot 90 90 67.1314>}
 ]
 spine_02 [{shapeType=Capsule, length=14.1596, radius=10.2002, shapeName=spine_02_capsule}]
 spine_03 [{shapeType=Capsule, length=13.5763, radius=11.3042, shapeName=spine_03_capsule, center=<0 1.2516 0>}]
 spine_04 [{shapeType=Capsule, length=19.7137, radius=13.6257, shapeName=spine_04_capsule, center=<0 2.4161 0>}]
 spine_05 [{shapeType=Capsule, length=27.0542, radius=11.794, shapeName=spine_05_capsule, center=<0 1.6185 0>}]
 thigh_l [{shapeType=Capsule, length=30.1221, radius=11.1367, shapeName=thigh_l_capsule, center=<-18.4993 0 0>, rotation=<rot 90 0 0>}]
 thigh_r [{shapeType=Capsule, length=30.1221, radius=11.1367, shapeName=thigh_r_capsule, center=<18.4993 0 0>, rotation=<rot 90 0 0>}]
 upperarm_l [{shapeType=Capsule, length=17.6233, radius=9.4151, shapeName=upperarm_l_capsule, center=<14.4803 2.768 0>, rotation=<rot 90 0 0>}]
 upperarm_r [{shapeType=Capsule, length=13.8767, radius=10.483, shapeName=upperarm_r_capsule, center=<-18.4993 -4.9305 0>, rotation=<rot 90 0 0>}]
}
constraints [
 {
  bone1Name=spine_05
  bone2Name=spine_04
  swing1LimitDegrees=45
  swing1Motion=Free
  swing2LimitDegrees=45
  swing2Motion=Free
  twistLimitDegrees=45
  twistMotion=Free
 }
 {
  bone1Name=clavicle_l
  bone2Name=spine_05
  swing1LimitDegrees=45
  swing1Motion=Free
  swing2LimitDegrees=45
  swing2Motion=Free
  twistLimitDegrees=45
  twistMotion=Free
 }
 {
  bone1Name=spine_04
  bone2Name=spine_03
  swing1LimitDegrees=45
  swing1Motion=Free
  swing2LimitDegrees=45
  swing2Motion=Free
  twistLimitDegrees=45
  twistMotion=Free
 }
 {
  bone1Name=upperarm_l
  bone2Name=clavicle_l
  swing1LimitDegrees=45
  swing1Motion=Free
  swing2LimitDegrees=45
  swing2Motion=Free
  twistLimitDegrees=45
  twistMotion=Free
 }
 {
  bone1Name=spine_02
  bone2Name=pelvis
  swing1LimitDegrees=45
  swing1Motion=Free
  swing2LimitDegrees=45
  swing2Motion=Free
  twistLimitDegrees=45
  twistMotion=Free
 }
 {
  bone1Name=lowerarm_l
  bone2Name=upperarm_l
  swing1LimitDegrees=45
  swing1Motion=Free
  swing2LimitDegrees=45
  swing2Motion=Free
  twistLimitDegrees=45
  twistMotion=Free
 }
 {
  bone1Name=clavicle_r
  bone2Name=spine_05
  swing1LimitDegrees=45
  swing1Motion=Free
  swing2LimitDegrees=45
  swing2Motion=Free
  twistLimitDegrees=45
  twistMotion=Free
 }
 {
  bone1Name=upperarm_r
  bone2Name=clavicle_r
  swing1LimitDegrees=45
  swing1Motion=Free
  swing2LimitDegrees=45
  swing2Motion=Free
  twistLimitDegrees=45
  twistMotion=Free
 }
 {
  bone1Name=lowerarm_r
  bone2Name=upperarm_r
  swing1LimitDegrees=45
  swing1Motion=Free
  swing2LimitDegrees=45
  swing2Motion=Free
  twistLimitDegrees=45
  twistMotion=Free
 }
 {
  bone1Name=thigh_l
  bone2Name=pelvis
  swing1LimitDegrees=45
  swing1Motion=Free
  swing2LimitDegrees=45
  swing2Motion=Free
  twistLimitDegrees=45
  twistMotion=Free
 }
 {
  bone1Name=calf_l
  bone2Name=thigh_l
  swing1LimitDegrees=45
  swing1Motion=Free
  swing2LimitDegrees=45
  swing2Motion=Free
  twistLimitDegrees=45
  twistMotion=Free
 }
 {
  bone1Name=thigh_r
  bone2Name=pelvis
  swing1LimitDegrees=45
  swing1Motion=Free
  swing2LimitDegrees=45
  swing2Motion=Free
  twistLimitDegrees=45
  twistMotion=Free
 }
 {
  bone1Name=calf_r
  bone2Name=thigh_r
  swing1LimitDegrees=45
  swing1Motion=Free
  swing2LimitDegrees=45
  swing2Motion=Free
  twistLimitDegrees=45
  twistMotion=Free
 }
 {
  bone1Name=spine_03
  bone2Name=spine_02
  swing1LimitDegrees=45
  swing1Motion=Free
  swing2LimitDegrees=45
  swing2Motion=Free
  twistLimitDegrees=45
  twistMotion=Free
 }
 {
  bone1Name=cape_l_MSTR
  bone2Name=clavicle_l
  swing1LimitDegrees=45
  swing1Motion=Free
  swing2LimitDegrees=45
  swing2Motion=Free
  twistLimitDegrees=45
  twistMotion=Free
 }
 {bone1Name=cape_01_l, bone2Name=cape_l_MSTR, swing1LimitDegrees=30, swing2LimitDegrees=45, twistLimitDegrees=0}
 {bone1Name=cape_02_l, bone2Name=cape_01_l, swing1LimitDegrees=20, swing2LimitDegrees=45, twistLimitDegrees=0}
 {bone1Name=cape_03_l, bone2Name=cape_02_l, swing1LimitDegrees=20, swing2LimitDegrees=45, twistLimitDegrees=0}
 {bone1Name=cape_04_l, bone2Name=cape_03_l, swing1LimitDegrees=15, swing2LimitDegrees=45, twistLimitDegrees=0}
 {bone1Name=cape_05_l, bone2Name=cape_04_l, swing1LimitDegrees=15, swing2LimitDegrees=20, twistLimitDegrees=0}
 {bone1Name=cape_06_l, bone2Name=cape_05_l, swing1LimitDegrees=20, swing2LimitDegrees=20, twistLimitDegrees=0}
 {
  bone1Name=cape_r_MSTR
  bone2Name=clavicle_r
  swing1LimitDegrees=45
  swing1Motion=Free
  swing2LimitDegrees=45
  swing2Motion=Free
  twistLimitDegrees=45
  twistMotion=Free
 }
 {
  bone1Name=cape_mid_l_MSTR
  bone2Name=spine_05
  swing1LimitDegrees=45
  swing1Motion=Free
  swing2LimitDegrees=45
  swing2Motion=Free
  twistLimitDegrees=45
  twistMotion=Free
 }
 {bone1Name=cape_mid_01_l, bone2Name=cape_mid_l_MSTR, swing1LimitDegrees=0, swing2LimitDegrees=45, twistLimitDegrees=0}
 {bone1Name=cape_mid_02_l, bone2Name=cape_mid_01_l, swing1LimitDegrees=10, swing2LimitDegrees=45, twistLimitDegrees=0}
 {bone1Name=cape_mid_03_l, bone2Name=cape_mid_02_l, swing1LimitDegrees=10, swing2LimitDegrees=45, twistLimitDegrees=0}
 {bone1Name=cape_mid_04_l, bone2Name=cape_mid_03_l, swing1LimitDegrees=10, swing2LimitDegrees=45, twistLimitDegrees=0}
 {bone1Name=cape_mid_05_l, bone2Name=cape_mid_04_l, swing1LimitDegrees=20, swing2LimitDegrees=20, twistLimitDegrees=0}
 {bone1Name=cape_mid_06_l, bone2Name=cape_mid_05_l, swing1LimitDegrees=20, swing2LimitDegrees=20, twistLimitDegrees=0}
 {
  bone1Name=cape_mid_r_MSTR
  bone2Name=spine_05
  swing1LimitDegrees=45
  swing1Motion=Free
  swing2LimitDegrees=45
  swing2Motion=Free
  twistLimitDegrees=45
  twistMotion=Free
 }
 {bone1Name=cape_mid_01_r, bone2Name=cape_mid_r_MSTR, swing1LimitDegrees=0, swing2LimitDegrees=45, twistLimitDegrees=0}
 {bone1Name=cape_mid_02_r, bone2Name=cape_mid_01_r, swing1LimitDegrees=10, swing2LimitDegrees=45, twistLimitDegrees=0}
 {bone1Name=cape_mid_03_r, bone2Name=cape_mid_02_r, swing1LimitDegrees=10, swing2LimitDegrees=45, twistLimitDegrees=0}
 {bone1Name=cape_mid_04_r, bone2Name=cape_mid_03_r, swing1LimitDegrees=10, swing2LimitDegrees=45, twistLimitDegrees=0}
 {bone1Name=cape_mid_05_r, bone2Name=cape_mid_04_r, swing1LimitDegrees=20, swing2LimitDegrees=20, twistLimitDegrees=0}
 {bone1Name=cape_mid_06_r, bone2Name=cape_mid_05_r, swing1LimitDegrees=20, swing2LimitDegrees=20, twistLimitDegrees=0}
 {
  bone1Name=foot_l
  bone2Name=calf_l
  swing1LimitDegrees=45
  swing1Motion=Free
  swing2LimitDegrees=45
  swing2Motion=Free
  twistLimitDegrees=45
  twistMotion=Free
 }
 {
  bone1Name=foot_r
  bone2Name=calf_r
  swing1LimitDegrees=45
  swing1Motion=Free
  swing2LimitDegrees=45
  swing2Motion=Free
  twistLimitDegrees=45
  twistMotion=Free
 }
 {
  bone1Name=pelvis
  bone2Name=root
  swing1LimitDegrees=45
  swing1Motion=Free
  swing2LimitDegrees=45
  swing2Motion=Free
  twistLimitDegrees=45
  twistMotion=Free
 }
 {bone1Name=cape_01_r, bone2Name=cape_r_MSTR, swing1LimitDegrees=10, swing2LimitDegrees=45, twistLimitDegrees=0}
 {bone1Name=cape_02_r, bone2Name=cape_01_r, swing1LimitDegrees=20, swing2LimitDegrees=45, twistLimitDegrees=0}
 {bone1Name=cape_03_r, bone2Name=cape_02_r, swing1LimitDegrees=20, swing2LimitDegrees=45, twistLimitDegrees=0}
 {bone1Name=cape_04_r, bone2Name=cape_03_r, swing1LimitDegrees=25, swing2LimitDegrees=45, twistLimitDegrees=0}
 {bone1Name=cape_05_r, bone2Name=cape_04_r, swing1LimitDegrees=45, swing2LimitDegrees=45, twistLimitDegrees=0}
 {bone1Name=cape_06_r, bone2Name=cape_05_r, swing1LimitDegrees=45, swing2LimitDegrees=45, twistLimitDegrees=0}
 {
  bone1Name=cape_02_l
  bone2Name=cape_mid_02_l
  swing1LimitDegrees=45
  swing1Motion=Free
  swing2LimitDegrees=45
  swing2Motion=Free
  twistLimitDegrees=45
  twistMotion=Free
 }
 {
  bone1Name=cape_06_l
  bone2Name=cape_mid_06_l
  swing1LimitDegrees=45
  swing1Motion=Free
  swing2LimitDegrees=45
  swing2Motion=Free
  twistLimitDegrees=45
  twistMotion=Free
 }
 {
  bone1Name=cape_05_l
  bone2Name=cape_mid_05_l
  swing1LimitDegrees=45
  swing1Motion=Free
  swing2LimitDegrees=45
  swing2Motion=Free
  twistLimitDegrees=45
  twistMotion=Free
 }
 {
  bone1Name=cape_04_l
  bone2Name=cape_mid_04_l
  swing1LimitDegrees=45
  swing1Motion=Free
  swing2LimitDegrees=45
  swing2Motion=Free
  twistLimitDegrees=45
  twistMotion=Free
 }
 {
  bone1Name=cape_mid_05_l
  bone2Name=cape_mid_05_r
  swing1LimitDegrees=45
  swing1Motion=Free
  swing2LimitDegrees=45
  swing2Motion=Free
  twistLimitDegrees=45
  twistMotion=Free
 }
 {
  bone1Name=cape_05_r
  bone2Name=cape_mid_05_r
  swing1LimitDegrees=45
  swing1Motion=Free
  swing2LimitDegrees=45
  swing2Motion=Free
  twistLimitDegrees=45
  twistMotion=Free
 }
 {
  bone1Name=cape_06_r
  bone2Name=cape_mid_06_r
  swing1LimitDegrees=45
  swing1Motion=Free
  swing2LimitDegrees=45
  swing2Motion=Free
  twistLimitDegrees=45
  twistMotion=Free
 }
]
